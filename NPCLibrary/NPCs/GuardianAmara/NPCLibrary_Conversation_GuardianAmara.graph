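# Conversation database for the Guardian Amara NPC, providing her greeting dialogue and access to shared NPC conversation options.
# A single entry point leads to a speech node with a default line, overridden by contextual lines for a first-ever conversation or specific player outfits.
# Curve-table rows set the priority of those contextual lines.
# The speech then branches through link nodes to the shared NPC library quests and basic services conversations by entry tag.
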ConversationDatabase /NPCLibrary/NPCs/GuardianAmara/NPCLibrary_Conversation_GuardianAmara
entry_points [{tag=Conversation.Entry.GuardianAmara, destinations [#1]}]
nodes [
 ConversationEntryPointNode_7 {
  id=#1
  class=ConversationEntryPointNode
  outputs [#2]
  properties {entryTag {tagName=Conversation.Entry.GuardianAmara}}
 }
 Speech_C_4 {
  id=#2
  class=Speech_C
  outputs [#3, #4]
  properties {
   nodeName=""
   generalConfig {
    defaultMessage=May the Island walk with you.
    contextualMessages [
     {
      message="My secrets are my own, and what I seek is not your business."
      contextRequirements [{requirement {class=FortControllerRequirement_IsFirstEverConversationWithNPC}}]
      priority {curve {rowName=Tandem.Speech.Priority.FirstEverInteraction}}
     }
     {
      message=I hear you've turned over a new leaf. Good.
      contextRequirements [{requirement {
         class=FortControllerRequirement_HasCID
         properties {allowedCIds [/BRCosmetics/Athena/Items/Cosmetics/Characters/CID_397_Athena_Commando_F_TreasureHunterFashion]}
        }}]
      priority {curve {rowName=Tandem.Speech.Priority.OutfitReaction}}
     }
     {
      message="Of these 'explorers', you are the one I trust to respect this place."
      contextRequirements [{requirement {class=FortControllerRequirement_HasCID, properties {allowedCIds [/BRCosmetics/Athena/Items/Cosmetics/Characters/Character_ZirconSweep]}}}]
      priority {curve {rowName=Tandem.Speech.Priority.OutfitReaction}}
     }
     {
      message="I've heard of you. Nothing good, of course."
      contextRequirements [{requirement {class=FortControllerRequirement_HasCID, properties {allowedCIds [&Character_InnovatorSand]}}}]
      priority {curve {rowName=Tandem.Speech.Priority.OutfitReaction}}
     }
     {
      message="As long as your philosophy is 'look, don't touch', we will not have a problem."
      contextRequirements [{requirement {class=FortControllerRequirement_HasCID, properties {allowedCIds [/BRCosmetics/Athena/Items/Cosmetics/Characters/Character_CrispRover]}}}]
      priority {curve {rowName=Tandem.Speech.Priority.OutfitReaction}}
     }
     {
      message=I understand the weight of inherited responsibility.
      contextRequirements [{requirement {class=FortControllerRequirement_HasCID, properties {allowedCIds [/BRCosmetics/Athena/Items/Cosmetics/Characters/Character_SilentTempo]}}}]
      priority {curve {rowName=Tandem.Speech.Priority.OutfitReaction}}
     }
    ]
   }
  }
 }
 ConversationLinkNode_0 {id=#4, class=ConversationLinkNode, properties {remoteEntryTag {tagName=Conversation.Entry.BasicServices}}}
 ConversationLinkNode_1 {id=#3, class=ConversationLinkNode, properties {remoteEntryTag {tagName=Conversation.Entry.NPCLibrary.Quests}}}
]
